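# MSS_EVO_Initialize is a one-shot stereo MetaSound source, apparently for an EVO initialize sound.
# On play, the trigger passes through an input-driven delay into a stereo wave player, whose wave comes from a graph variable with no asset shown.
# Left and right outputs are scaled by the MasterVolume input, and playback completion fires OnFinished.
# Concurrency allows one instance, stopping the oldest, and virtualization is disabled.
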MetaSoundSource /Game/Sounds/EVO/MSS_EVO_Initialize
inputs [
 Delay Time {type=Time}
 MasterVolume {type=Float}
 UE.Source.OnPlay {type=Trigger}
]
interfaces [UE.OutputFormat.Stereo:1.0, UE.Source.OneShot:1.0, UE.Source:1.1]
outputs [
 "UE.OutputFormat.Stereo.Audio:0" {type=Audio}
 "UE.OutputFormat.Stereo.Audio:1" {type=Audio}
 UE.Source.OneShot.OnFinished {type=Trigger}
]
pages [Default {
  input_defaults {Delay Time=0, MasterVolume=1, UE.Source.OnPlay=null}
  nodes [
   TriggerDelay_0 = UE.Trigger Delay(Delay Time=^Delay Time, In=^UE.Source.OnPlay)
   WavePlayer_0 = UE.Wave Player:Stereo(Play=@TriggerDelay_0.Out, Wave Asset=VariableAccessor.WaveAsset(Variable=InitVariable.WaveAsset().Variable).Value) {also {version [1,1]}}
   Multiply_0 = UE.Multiply:Audio by Float(AdditionalOperands=^MasterVolume, PrimaryOperand=@WavePlayer_0.Out Right)
   Multiply_1 = UE.Multiply:Audio by Float(AdditionalOperands=^MasterVolume, PrimaryOperand=@WavePlayer_0.Out Left)
  ]
  links [
   {dst="@UE.OutputFormat.Stereo.Audio:1", src=Multiply_0.Out}
   {dst="@UE.OutputFormat.Stereo.Audio:0", src=Multiply_1.Out}
   {dst="@UE.Source.OneShot.OnFinished", src=WavePlayer_0.On Finished}
  ]
 }]
properties {
 assetImportData=&self:AssetImportData
 bOverrideConcurrency=true
 numChannels=2
 outputFormat=Stereo
 sampleRate=48000
 soundClassObject=/Game/Sounds/FortSoundClassesAndModes/Master
 soundSubmixObject=/Game/Sounds/SubMixes/Fortnite_Main
 virtualizationMode=Disabled
 volume=1.007
 concurrencyOverrides {maxCount=1, resolutionRule=StopOldest}
}
